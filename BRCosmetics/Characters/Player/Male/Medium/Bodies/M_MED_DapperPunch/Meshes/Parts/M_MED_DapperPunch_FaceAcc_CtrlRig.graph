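# Control Rig for the DapperPunch character's face accessory mesh that correctively poses the first hair tail bone as the head moves.
# It aims the bone at spine_05, then layers rotation and translation offsets weighted by four head spherical pose readers, three spine-relative and one root-relative, used inverted.
# Crouching blends between crouched and standing corrections, Is Melee and pistol suppresses the crouched ones, and Targeting adds a standing rotation.
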ControlRigBlueprint /BRCosmetics/Characters/Player/Male/Medium/Bodies/M_MED_DapperPunch/Meshes/Parts/M_MED_DapperPunch_FaceAcc_CtrlRig
graphs [
 Multiply_7_1_1_ContainedGraph {nodes [
   n1 = RigVMFunctionEntryNode()
   RigVMFunctionReturnNode(Result=&RigVMFunction_MathDoubleMul(B=@n1.C, A=&RigVMFunction_MathDoubleMul(A=@n1.A, B=@n1.B).Result).Result)
  ]}
 Multiply_7_2_ContainedGraph {nodes [
   n1 = RigVMFunctionEntryNode()
   RigVMFunctionReturnNode(Result=&RigVMFunction_MathDoubleMul(B=@n1.C, A=&RigVMFunction_MathDoubleMul(A=@n1.A, B=@n1.B).Result).Result)
  ]}
 RigVMFunctionLibrary {links [], nodes []}
 RigVMModel {nodes [
   RigUnit_BeginExecution = &RigUnit_BeginExecution()
   SequenceAggregate = &RigVMFunction_Sequence(ExecuteContext=@RigUnit_BeginExecution.ExecutePin)
   n1 = &RigUnit_SphericalPoseReader(
    ActiveRegionScaleFactors=(PositiveWidth=1,NegativeWidth=1,PositiveHeight=1,NegativeHeight=1)
    ActiveRegionSize=0.05
    Debug=(bDrawDebug=false,bDraw2D=false,bDrawLocalAxes=false,DebugScale=25,DebugSegments=20,DebugThickness=0.25)
    DriverAxis=(X=1,Y=0,Z=0)
    DriverItem=(Type=Bone,Name="head")
    FalloffRegionScaleFactors=(PositiveWidth=1,NegativeWidth=1,PositiveHeight=1,NegativeHeight=1)
    FalloffSize=0.4
    OptionalParentItem=(Type=Bone,Name="spine_05")
    RotationOffset=(X=-55,Y=0,Z=30)
    ExecutePin=@SequenceAggregate.A
   )
   > n2 = &RigUnit_SphericalPoseReader(
    ActiveRegionScaleFactors=(PositiveWidth=1,NegativeWidth=1,PositiveHeight=1,NegativeHeight=1)
    ActiveRegionSize=0.05
    Debug=(bDrawDebug=false,bDraw2D=false,bDrawLocalAxes=false,DebugScale=25,DebugSegments=20,DebugThickness=0.25)
    DriverAxis=(X=1,Y=0,Z=0)
    DriverItem=(Type=Bone,Name="head")
    FalloffRegionScaleFactors=(PositiveWidth=1,NegativeWidth=1,PositiveHeight=1,NegativeHeight=1)
    FalloffSize=0.25
    OptionalParentItem=(Type=Bone,Name="spine_05")
    RotationOffset=(X=-35,Y=0,Z=40)
   )
   > n3 = &RigUnit_SphericalPoseReader(
    ActiveRegionScaleFactors=(PositiveWidth=1,NegativeWidth=1,PositiveHeight=1,NegativeHeight=1)
    ActiveRegionSize=0.05
    Debug=(bDrawDebug=true,bDraw2D=false,bDrawLocalAxes=false,DebugScale=25,DebugSegments=20,DebugThickness=0.25)
    DriverAxis=(X=1,Y=0,Z=0)
    DriverItem=(Type=Bone,Name="head")
    FalloffRegionScaleFactors=(PositiveWidth=1,NegativeWidth=1,PositiveHeight=1,NegativeHeight=1)
    FalloffSize=0.2
    OptionalParentItem=(Type=Bone,Name="spine_05")
    RotationOffset=(X=85,Y=0,Z=-80)
   )
   > &RigUnit_AimConstraintLocalSpaceOffset(
    AdvancedSettings=(DebugSettings=(bEnabled=false,Scale=100,WorldOffset=(Rotation=(X=0,Y=0,Z=0,W=1),Translation=(X=0,Y=0,Z=0),Scale3D=(X=1,Y=1,Z=1))),RotationOrderForFilter=XZY)
    Child=(Type=Bone,Name="dyn_hair_tail_1")
    Filter=(bX=true,bY=true,bZ=true)
    Parents=((Item=(Type=Bone,Name="spine_05"),Weight=1))
    UpAxis=(X=0,Y=-1,Z=0)
    Weight=1
    WorldUp=(Target=(X=0,Y=0,Z=0),Kind=Direction,Space=(Type=Bone,Name="root"))
   )
   > &RigUnit_ModifyTransforms(
    ItemToModify=((Item=(Type=Bone,Name="dyn_hair_tail_1"),Transform=(Rotation=(X=0,Y=0,Z=0,W=1),Translation=(X=-2,Y=-3,Z=0),Scale3D=(X=1,Y=1,Z=1))))
    WeightMinimum=0
    ItemToModify.0.Transform.Rotation=&RigVMFunction_MathQuaternionFromEuler(Euler=(X=0,Y=0,Z=40)).Result
    Weight=@n1.OutputParam
   )
   n4 = &RigVMFunction_MathDoubleSub(A=1, B=RigVMVariableNode() {variable=Is Melee and pistol}.Value)
   n5 = RigVMVariableNode() {variable=Crouching}
   Multiply_7 = RigVMAggregateNode(B=@n2.OutputParam, A=@n4.Result, C=@n5.Value)
   > &RigUnit_ModifyTransforms(
    ItemToModify=((Item=(Type=Bone,Name="dyn_hair_tail_1"),Transform=(Rotation=(X=0,Y=0,Z=0,W=1),Translation=(X=0,Y=0,Z=0),Scale3D=(X=1,Y=1,Z=1))))
    WeightMinimum=0
    Weight=@Multiply_7.Result
    ItemToModify.0.Transform.Rotation=&RigVMFunction_MathQuaternionFromEuler(Euler=(X=-20,Y=0,Z=-1)).Result
   )
   Multiply_7_1 = RigVMAggregateNode(B=@n3.OutputParam, A=@n4.Result, C=@n5.Value)
   > &RigUnit_ModifyTransforms(
    ItemToModify=((Item=(Type=Bone,Name="dyn_hair_tail_1"),Transform=(Rotation=(X=0,Y=0,Z=0,W=1),Translation=(X=0,Y=0,Z=0),Scale3D=(X=1,Y=1,Z=1))))
    WeightMinimum=0
    Weight=@Multiply_7_1.Result
    ItemToModify.0.Transform.Rotation=&RigVMFunction_MathQuaternionFromEuler(Euler=(X=-20,Y=0,Z=-2)).Result
   )
   n6 = &RigUnit_SphericalPoseReader(
    ActiveRegionScaleFactors=(PositiveWidth=20,NegativeWidth=20,PositiveHeight=1,NegativeHeight=1)
    ActiveRegionSize=0.01
    Debug=(bDrawDebug=false,bDraw2D=false,bDrawLocalAxes=false,DebugScale=25,DebugSegments=20,DebugThickness=0.25)
    DriverAxis=(X=0,Y=1,Z=0)
    DriverItem=(Type=Bone,Name="head")
    FalloffRegionScaleFactors=(PositiveWidth=3,NegativeWidth=3,PositiveHeight=1,NegativeHeight=3)
    FalloffSize=0.3
    OptionalParentItem=(Type=Bone,Name="root")
    RotationOffset=(X=90,Y=0,Z=30)
    ExecutePin=@SequenceAggregate.B
   )
   n7 = &RigVMFunction_MathDoubleSub(A=1, B=@n5.Value)
   Multiply = &RigVMFunction_MathDoubleMul(A=@n1.OutputParam, B=@n7.Result)
   > &RigUnit_ModifyTransforms(
    ItemToModify=((Item=(Type=Bone,Name="dyn_hair_tail_1"),Transform=(Rotation=(X=0,Y=0,Z=0,W=1),Translation=(X=-1,Y=1,Z=2),Scale3D=(X=1,Y=1,Z=1))))
    WeightMinimum=0
    Weight=@Multiply.Result
    ItemToModify.0.Transform.Rotation=&RigVMFunction_MathQuaternionFromEuler(Euler=(X=0,Y=0,Z=20)).Result
   )
   > &RigUnit_ModifyTransforms(
    ItemToModify=((Item=(Type=Bone,Name="dyn_hair_tail_1"),Transform=(Rotation=(X=0,Y=0,Z=0,W=1),Translation=(X=0,Y=0,Z=0),Scale3D=(X=1,Y=1,Z=1))))
    WeightMinimum=0
    Weight=&RigVMFunction_MathDoubleMul(A=@n3.OutputParam, B=@n7.Result).Result
    ItemToModify.0.Transform.Rotation=&RigVMFunction_MathQuaternionFromEuler(Euler=(X=0,Y=0,Z=-4)).Result
   )
   Subtract = &RigVMFunction_MathDoubleSub(A=1, B=@n6.OutputParam)
   > &RigUnit_ModifyTransforms(
    ItemToModify=((Item=(Type=Bone,Name="dyn_hair_tail_1"),Transform=(Rotation=(X=0,Y=0,Z=0,W=1),Translation=(X=0,Y=0,Z=0),Scale3D=(X=1,Y=1,Z=1))))
    WeightMinimum=0
    Weight=&RigVMFunction_MathDoubleMul(B=@Subtract.Result, A=@n7.Result).Result
    ItemToModify.0.Transform.Rotation=&RigVMFunction_MathQuaternionFromEuler(Euler=(X=0,Y=0,Z=-9)).Result
   )
   n8 = RigVMVariableNode() {variable=Targeting}
   > &RigUnit_ModifyTransforms(
    ItemToModify=((Item=(Type=Bone,Name="dyn_hair_tail_1"),Transform=(Rotation=(X=0,Y=0,Z=0,W=1),Translation=(X=0,Y=0,Z=0),Scale3D=(X=1,Y=1,Z=1))))
    WeightMinimum=0
    Weight=&RigVMFunction_MathDoubleMul(B=&RigVMFunction_MathDoubleMul(A=@n7.Result, B=@n8.Value).Result, A=@n1.OutputParam).Result
    ItemToModify.0.Transform.Rotation=&RigVMFunction_MathQuaternionFromEuler(Euler=(X=0,Y=0,Z=-9)).Result
   )
  ]}
]
hierarchy [
 root {loc=<0 0 0>, rot=<0 0 0 1>}
 pelvis {parent=root, loc=<0 -0.7071 93.1191>, rot=<0.0127 0.707 -0.0127 -0.707>}
 spine_01 {parent=pelvis, loc=<6.7494 -0.6388 0>, rot=<0 0 0.034 -0.9994>}
 spine_02 {parent=spine_01, loc=<7.2518 -0.035 0>, rot=<0 0 0.0238 -0.9997>}
 spine_03 {parent=spine_02, loc=<6.0104 0 0>, rot=<0 0 0.0149 -0.9999>}
 spine_04 {parent=spine_03, loc=<8.1024 0 0>, rot=<0 0 -0.0132 -0.9999>}
 spine_05 {parent=spine_04, loc=<10.0146 0 0>, rot=<0 0 -0.0015 -1>}
 neck_01 {parent=spine_05, loc=<11.3138 0.1529 0>, rot=<0 0 -0.2003 -0.9797>}
 neck_02 {parent=neck_01, loc=<4.5627 0 0>, rot=<0 0 0.0504 -0.9987>}
 head {parent=neck_02, loc=<5.3914 0 0>, rot=<0 0 0.115 -0.9934>}
 dyn_hair_tail_1 {parent=head, loc=<3.8653 -3.8177 0>, rot=<0.196 -0.9806 0 0>}
 dyn_hair_tail_2 {parent=dyn_hair_tail_1, loc=<5.3116 0 0>, rot=<0 0 0.2644 -0.9644>}
 dyn_hair_tail_3 {parent=dyn_hair_tail_2, loc=<5.7152 0 0>, rot=<0 0 0 1>}
 dyn_hair_base {parent=head, loc=<13.3463 0.4458 -0.1486>, rot=<0.0238 -0.0136 -0.7091 -0.7046>}
 dyn_hair_back_1 {parent=head, loc=<12.2584 -3.1822 -6.1009>, rot=<-0.4889 0.0474 -0.1022 -0.865>}
 dyn_hair_middle_1 {parent=head, loc=<11.8067 5.6245 6.8631>, rot=<-0.0346 0.3587 -0.4754 -0.8026>}
 dyn_hair_middle_2 {parent=head, loc=<7.6464 4.3819 6.8905>, rot=<-0.0235 -0.5152 0.8169 0.2582>}
 dyn_hair_middle_3 {parent=head, loc=<16.0583 5.687 1.0805>, rot=<-0.3896 0.2182 -0.2766 -0.8509>}
 dyn_hair_front_1 {parent=head, loc=<8.936 10.6773 2.415>, rot=<0.1588 -0.4109 0.8935 0.0871>}
 dyn_hair_front_2 {parent=dyn_hair_front_1, loc=<3.6448 0.2877 -0.0936>, rot=<-0.0805 -0.0996 0.0081 -0.9917>}
 dyn_hair_front_3 {parent=head, loc=<13.2007 9.9149 1.0816>, rot=<-0.421 0.2901 -0.5329 -0.6743>}
 dyn_hair_front_4 {parent=head, loc=<10.6732 10.1762 4.3212>, rot=<0.2934 -0.4141 0.7436 0.4353>}
 dyn_hair_front_5 {parent=head, loc=<8.3723 8.3008 5.7598>, rot=<0.2719 -0.3806 0.7427 0.4793>}
]
variables [
 Crouching {type=double, default=0}
 Is Melee and pistol {type=double, default=0}
 Targeting {type=double, default=0}
]
