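# Physics asset for the face accessory mesh of the medium male DapperPunch player body, setting up dynamic hair bones for physical motion.
# Capsule bodies cover the hair, head, neck and upper spine, with free angular constraints linking head, neck and spine.
# A three-bone tail and a two-bone front lock chain from the head, while other hair bones attach directly to it.
# Angular limits keep the base nearly rigid, loose strands wider and twist tight.
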PhysicsAsset /BRCosmetics/Characters/Player/Male/Medium/Bodies/M_MED_DapperPunch/Meshes/Parts/M_MED_DapperPunch_FaceAcc_Physics
bodies {
 dyn_hair_back_1 [{shapeType=Capsule, length=2, radius=1.5, shapeName=dyn_hair_back_1_capsule, center=<0.5766 0 0>, rotation=<rot 90 0 0>}]
 dyn_hair_base [{shapeType=Capsule, length=2, radius=1.5, shapeName=dyn_hair_base_capsule, center=<0.5766 0 0>, rotation=<rot 90 0 0>}]
 dyn_hair_front_1 [{
   shapeType=Capsule
   length=2.1419
   radius=1.6726
   shapeName=dyn_hair_front_1_capsule
   center=<2.6291 0.1807 -0.0139>
   rotation=<rot 0 -86.0568 -89.7434>
  }]
 dyn_hair_front_2 [{shapeType=Capsule, length=2, radius=1.5, shapeName=dyn_hair_front_2_capsule, center=<0.5766 0 0>, rotation=<rot 90 0 0>}]
 dyn_hair_front_3 [{shapeType=Capsule, length=2, radius=1.5, shapeName=dyn_hair_front_3_capsule, center=<0.2 0 0>, rotation=<rot 90 0 0>}]
 dyn_hair_front_4 [{shapeType=Capsule, length=2, radius=1.5, shapeName=dyn_hair_front_4_capsule, center=<0.5766 0 0>, rotation=<rot 90 0 0>}]
 dyn_hair_front_5 [{shapeType=Capsule, length=2, radius=1.5, shapeName=dyn_hair_front_5_capsule, center=<0.5766 0 0>, rotation=<rot 90 0 0>}]
 dyn_hair_middle_1 [{shapeType=Capsule, length=2, radius=1.5, shapeName=dyn_hair_middle_1_capsule, center=<0.2 0 0>, rotation=<rot 90 0 0>}]
 dyn_hair_middle_2 [{shapeType=Capsule, length=2, radius=1.5, shapeName=dyn_hair_middle_2_capsule, center=<0.5766 0 0>, rotation=<rot 90 0 0>}]
 dyn_hair_middle_3 [{shapeType=Capsule, length=2, radius=1.5, shapeName=dyn_hair_middle_3_capsule, center=<0.2 0 0>, rotation=<rot 90 0 0>}]
 dyn_hair_tail_1 [{shapeType=Capsule, length=3, radius=3, shapeName=dyn_hair_tail_1_capsule}]
 dyn_hair_tail_2 [{shapeType=Capsule, length=0, radius=3, shapeName=dyn_hair_tail_2_capsule, center=<0.9606 0 0>}]
 dyn_hair_tail_3 [
  {shapeType=Capsule, length=4.0074, radius=5.026, shapeName=dyn_hair_tail_3_capsule, center=<4.0613 -0.1698 0>}
  {
   shapeType=Capsule
   length=3.3883
   radius=2
   shapeName=dyn_hair_tail_3_capsule
   center=<1.9853 2.06 3.2138>
   rotation=<rot -67.8621 -2.7308 21.5274>
  }
 ]
 head [{shapeType=Capsule, length=3, radius=3, shapeName=head_capsule}]
 neck_01 [{shapeType=Capsule, length=11.8308, radius=3.9309, shapeName=neck_01_capsule, rotation=<rot 90 0 0>}]
 spine_05 [
  {shapeType=Capsule, length=13.3996, radius=12.0102, shapeName=spine_05_capsule, center=<5.4792 2.7308 0>}
  {shapeType=Capsule, length=10, radius=5, shapeName=spine_05_capsule, center=<12.0001 -1.7957 0>}
 ]
}
constraints [
 {
  bone1Name=head
  bone2Name=neck_01
  swing1LimitDegrees=45
  swing1Motion=Free
  swing2LimitDegrees=45
  swing2Motion=Free
  twistLimitDegrees=45
  twistMotion=Free
 }
 {bone1Name=dyn_hair_tail_3, bone2Name=dyn_hair_tail_2, swing1LimitDegrees=25, swing2LimitDegrees=35, twistLimitDegrees=5}
 {bone1Name=dyn_hair_tail_1, bone2Name=head, swing1LimitDegrees=10, swing2LimitDegrees=10, twistLimitDegrees=0}
 {bone1Name=dyn_hair_tail_2, bone2Name=dyn_hair_tail_1, swing1LimitDegrees=25, swing2LimitDegrees=25, twistLimitDegrees=5}
 {
  bone1Name=neck_01
  bone2Name=spine_05
  swing1LimitDegrees=45
  swing1Motion=Free
  swing2LimitDegrees=45
  swing2Motion=Free
  twistLimitDegrees=45
  twistMotion=Free
 }
 {bone1Name=dyn_hair_front_1, bone2Name=head, swing1LimitDegrees=15, swing2LimitDegrees=15, twistLimitDegrees=5}
 {bone1Name=dyn_hair_front_2, bone2Name=dyn_hair_front_1, swing1LimitDegrees=35, swing2LimitDegrees=35, twistLimitDegrees=5}
 {bone1Name=dyn_hair_base, bone2Name=head, swing1LimitDegrees=2, swing2LimitDegrees=2, twistLimitDegrees=2}
 {bone1Name=dyn_hair_back_1, bone2Name=head, swing1LimitDegrees=35, swing2LimitDegrees=35, twistLimitDegrees=5}
 {bone1Name=dyn_hair_front_3, bone2Name=head, swing1LimitDegrees=35, swing2LimitDegrees=35, twistLimitDegrees=5}
 {bone1Name=dyn_hair_front_4, bone2Name=head, swing1LimitDegrees=35, swing2LimitDegrees=35, twistLimitDegrees=5}
 {bone1Name=dyn_hair_front_5, bone2Name=head, swing1LimitDegrees=35, swing2LimitDegrees=35, twistLimitDegrees=5}
 {bone1Name=dyn_hair_middle_1, bone2Name=head, swing1LimitDegrees=35, swing2LimitDegrees=35, twistLimitDegrees=5}
 {bone1Name=dyn_hair_middle_2, bone2Name=head, swing1LimitDegrees=35, swing2LimitDegrees=35, twistLimitDegrees=5}
 {bone1Name=dyn_hair_middle_3, bone2Name=head, swing1LimitDegrees=35, swing2LimitDegrees=35, twistLimitDegrees=5}
]
