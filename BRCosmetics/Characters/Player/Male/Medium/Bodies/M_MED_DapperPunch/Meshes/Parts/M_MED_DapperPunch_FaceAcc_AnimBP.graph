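# Post-process style AnimBP for the DapperPunch face-accessory part, driving hair-tail secondary motion on the player character.
# Copied parent pose runs through RigidBody, low-weight AnimDynamics, Control Rig, then additive tail-bone offsets per movement state.
# EventGraph derives state flags, eased rig inputs, head rotation and gravity, reducing RigidBody weight during sprint, jump or fall.
# Head-rotation corrective chains, idle blend and back-hair AnimDynamics are not connected to the output.
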AnimBlueprint /BRCosmetics/Characters/Player/Male/Medium/Bodies/M_MED_DapperPunch/Meshes/Parts/M_MED_DapperPunch_FaceAcc_AnimBP
target_skeleton=&/M_MED_DapperPunch_FaceAcc_Skeleton
graphs [
 AnimGraph {nodes [
   n1 = Get gravity_override() -> (gravity_override:Vector)
   Anim Dynamics - Bone: dyn_hair_back_base!(ComponentPose, Alpha=1, GravityOverride:Vector=@n1, ExternalForce="0,0,0") -> (Pose) {
    angularDampingOverride=1
    angularSpringConstant=20000
    bAngularSpring=true
    bGravityOverrideInSimSpace=true
    bOverrideAngularDamping=true
    bUseGravityOverride=true
    gravityScale=0.35
    simulationSpace=RootRelative
    alphaBoolBlend {blendOption=Linear}
    boundBone {boneName=dyn_hair_back_base}
    componentLinearAccScale=<1 1 1>
    componentLinearVelScale=<1 1 1>
    gravityOverride {z=-980}
    physicsBodyDefinitions [{
      sphereCollisionRadius=0
      boundBone {boneName=dyn_hair_back_base}
      boxExtents=<10 20 20>
      constraintSetup {angularLimitsMax=<5 5 5>, angularLimitsMin=<-5 -5 -5>, angularTarget=<1 0 0>}
      localJointOffset=<0.05 0 0>
     }]
    relativeSpaceBone {boneName=head}
   }
   n2 = "Transform (Modify) Bone - Bone: dyn_hair_tail_1"(
    ComponentPose=Local To Component(LocalPose=Use cached pose 'SavedPose23283'())
    Alpha=Get head_fwdBwd_rot() -> (head_fwdBwd_rot:double)
    Translation="-0.471,-0.119,0.000"
    Rotation="-0.029,-6.960,0.009"
    Scale="1,1,1"
   ) {
    rotationMode=BMM_Additive
    rotationSpace=BCS_BoneSpace
    translationMode=BMM_Additive
    translationSpace=BCS_BoneSpace
    alphaBoolBlend {blendOption=Linear}
    alphaScaleBiasClamp {bMapRange=true, inRange {max=-47}}
    boneToModify {boneName=dyn_hair_tail_1}
    rotation=<rot -0.029 -6.96 0.009>
    translation=<-0.471 -0.119>
   }
   n3 = "Transform (Modify) Bone - Bone: dyn_hair_tail_1"(ComponentPose=@n2, Alpha=Get head_fwdBwd_rot(), Translation="-0.324,-0.283,0.000", Rotation="0.003,3.839,-0.003", Scale="1,1,1") {
    rotationMode=BMM_Additive
    rotationSpace=BCS_BoneSpace
    translationMode=BMM_Additive
    translationSpace=BCS_BoneSpace
    alphaBoolBlend {blendOption=Linear}
    alphaScaleBiasClamp {bMapRange=true, inRange {max=-8, min=-35}}
    boneToModify {boneName=dyn_hair_tail_1}
    rotation=<rot 0.003 3.8391 -0.003>
    translation=<-0.324 -0.283>
   }
   n4 = "Transform (Modify) Bone - Bone: dyn_hair_tail_1"(ComponentPose=@n3, Alpha=Get head_fwdBwd_rot(), Translation="0,0,0", Rotation="0.000,-8.047,-0.011", Scale="1,1,1") {
    rotationMode=BMM_Additive
    rotationSpace=BCS_BoneSpace
    translationMode=BMM_Additive
    translationSpace=BCS_BoneSpace
    alphaBoolBlend {blendOption=Linear}
    alphaScaleBiasClamp {bMapRange=true, inRange {max=16, min=-8}}
    boneToModify {boneName=dyn_hair_tail_1}
    rotation=<rot 0 -8.0469 -0.011>
   }
   n5 = "Transform (Modify) Bone - Bone: dyn_hair_tail_1"(ComponentPose=@n4, Alpha=Get head_fwdBwd_rot(), Translation="0,0,0", Rotation="0.000,3.276,0.000", Scale="1,1,1") {
    rotationMode=BMM_Additive
    rotationSpace=BCS_BoneSpace
    translationMode=BMM_Additive
    translationSpace=BCS_BoneSpace
    alphaBoolBlend {blendOption=Linear}
    alphaScaleBiasClamp {bMapRange=true, inRange {max=19, min=16}}
    boneToModify {boneName=dyn_hair_tail_1}
    rotation=<rot 0 3.2761 0>
   }
   Blend Poses by bool(
    BlendPose_0=Component To Local(ComponentPose=@n5)
    BlendPose_1=Use cached pose 'SavedPose23283'()
    bActiveValue=Get is in idle() -> (is in idle:bool)
   ) {blendType=Linear}
   n6 = RigidBody!(
    ComponentPose=Local To Component(LocalPose=Copy Pose From Mesh() {bUseAttachedParent=true})
    Alpha=Get Alpha() -> (Alpha:float)
    OverrideWorldGravity:Vector=@n1
    ExternalForce
   ) -> (Pose) {
    bClampLinearTranslationLimitToRefPose=true
    bOverrideWorldGravity=true
    overridePhysicsAsset=&/M_MED_DapperPunch_FaceAcc_Physics
    simulationSpace=BaseBoneSpace
    alphaBoolBlend {blendOption=Linear}
    baseBoneRef {boneName=spine_05}
    componentAppliedLinearAccClamp=<3000 3000 3000>
    componentLinearAccScale=<0.3 0.3 0.3>
    componentLinearVelScale=<0.4 0.4 0.4>
    overrideWorldGravity {z=-980}
    simSpaceSettings {maxAngularAcceleration=0, maxLinearAcceleration=0, worldAlpha=1}
   }
   n7 = Anim Dynamics - Bone: dyn_hair_tail_1!(ComponentPose=@n6, Alpha=0.1, GravityOverride:Vector=@n1, ExternalForce="0,0,0") -> (Pose) {
    angularDampingOverride=1
    angularSpringConstant=20000
    bAngularSpring=true
    bGravityOverrideInSimSpace=true
    bOverrideAngularDamping=true
    bUseGravityOverride=true
    gravityScale=0.35
    simulationSpace=BoneRelative
    alphaBoolBlend {blendOption=Linear}
    boundBone {boneName=dyn_hair_tail_1}
    componentAppliedLinearAccClamp=<3000 3000 3000>
    componentLinearAccScale=<1 1 1>
    componentLinearVelScale=<1 1 1>
    gravityOverride {z=-980}
    physicsBodyDefinitions [{
      sphereCollisionRadius=0
      boundBone {boneName=dyn_hair_tail_1}
      boxExtents=<10 20 20>
      constraintSetup {angularLimitsMax=<5 10 4>, angularLimitsMin=<-5 -10 -15>, angularTarget=<1 0 0>}
      localJointOffset=<0.01 0 0>
     }]
    relativeSpaceBone {boneName=head}
   }
   n8 = Control Rig!(
    Source=Component To Local(ComponentPose=@n7)
    Alpha
    Is Melee and pistol:double=Get Melee and pistol Weapons() -> (Melee and pistol Weapons:double)
    Crouching:double=Get CrouchingAlpha() -> (CrouchingAlpha:double)
    Targeting:double=Get Targeting alpha() -> (Targeting alpha:double)
   ) -> (Pose) {
    alphaBoolBlend {blendOption=Linear}
    controlRigAssetReference {blueprintRigClass=&/M_MED_DapperPunch_FaceAcc_CtrlRig.M_MED_DapperPunch_FaceAcc_CtrlRig_C}
   }
   n9 = "Transform (Modify) Bone - Bone: dyn_hair_tail_1"!(
    ComponentPose=Local To Component(LocalPose=@n8)
    bAlphaBoolEnabled:bool=Get tactical sprinting and jumping or falling() -> (tactical sprinting and jumping or falling:bool)
    Translation="0.855,1.150,0.050"
    Rotation="0.000,-14.488,-0.000"
    Scale="1,1,1"
   ) -> (Pose) {
    alphaInputType=Bool
    rotationMode=BMM_Additive
    rotationSpace=BCS_BoneSpace
    translationMode=BMM_Additive
    translationSpace=BCS_BoneSpace
    alphaBoolBlend {blendInTime=0.2, blendOption=Linear, blendOutTime=0.2}
    boneToModify {boneName=dyn_hair_tail_1}
    rotation=<rot 0 -14.488 0>
    translation=<0.855 1.15 0.05>
   }
   n10 = "Transform (Modify) Bone - Bone: dyn_hair_tail_1"(
    ComponentPose=@n9
    bAlphaBoolEnabled=Get Crouching and moving() -> (Crouching and moving:bool)
    Translation="-0.549,1.123,0.000"
    Rotation="0.000,3.627,-0.000"
    Scale="1,1,1"
   ) {
    alphaInputType=Bool
    rotationMode=BMM_Additive
    rotationSpace=BCS_BoneSpace
    translationMode=BMM_Additive
    translationSpace=BCS_BoneSpace
    alphaBoolBlend {blendInTime=0.2, blendOption=Linear, blendOutTime=0.2}
    boneToModify {boneName=dyn_hair_tail_1}
    rotation=<rot 0 3.627 0>
    translation=<-0.549 1.123>
   }
   n11 = "Transform (Modify) Bone - Bone: dyn_hair_tail_1"(
    ComponentPose=@n10
    bAlphaBoolEnabled=Get Not crouching and moving() -> (Not crouching and moving:bool)
    Translation="-0.664,-0.135,0.000"
    Rotation="0,0,0"
    Scale="1,1,1"
   ) {
    alphaInputType=Bool
    rotationMode=BMM_Additive
    rotationSpace=BCS_BoneSpace
    translationMode=BMM_Additive
    translationSpace=BCS_BoneSpace
    alphaBoolBlend {blendInTime=0.2, blendOption=Linear, blendOutTime=0.2}
    boneToModify {boneName=dyn_hair_tail_1}
    translation=<-0.664 -0.135>
   }
   n12 = "Transform (Modify) Bone - Bone: dyn_hair_tail_1"(
    ComponentPose=@n11
    bAlphaBoolEnabled=Get bIsTacticalSprinting()
    Translation="-0.473,1.898,0.000"
    Rotation="-0.000,-18.994,0.000"
    Scale="1,1,1"
   ) {
    alphaInputType=Bool
    rotationMode=BMM_Additive
    rotationSpace=BCS_BoneSpace
    translationMode=BMM_Additive
    translationSpace=BCS_BoneSpace
    alphaBoolBlend {blendInTime=0.2, blendOption=Linear, blendOutTime=0.2}
    boneToModify {boneName=dyn_hair_tail_1}
    rotation=<rot 0 -18.994 0>
    translation=<-0.473 1.898>
   }
   n13 = "Transform (Modify) Bone - Bone: dyn_hair_tail_1"(ComponentPose=@n12, bAlphaBoolEnabled=Get bIsJumping(), Translation="0.656,-0.044,-0.960", Rotation="5.846,-12.892,-0.116", Scale="1,1,1") {
    alphaInputType=Bool
    rotationMode=BMM_Additive
    rotationSpace=BCS_BoneSpace
    translationMode=BMM_Additive
    translationSpace=BCS_BoneSpace
    alphaBoolBlend {blendInTime=0.3, blendOption=Linear, blendOutTime=0.4}
    boneToModify {boneName=dyn_hair_tail_1}
    rotation=<rot 5.846 -12.892 -0.116>
    translation=<0.656 -0.044 -0.96>
   }
   n14 = "Transform (Modify) Bone - Bone: dyn_hair_tail_1"(ComponentPose=@n13, bAlphaBoolEnabled=Get bIsFalling(), Translation="0.656,-0.044,-0.96", Rotation="5.846,-12.892,-0.116", Scale="1,1,1") {
    alphaInputType=Bool
    rotationMode=BMM_Additive
    rotationSpace=BCS_BoneSpace
    translationMode=BMM_Additive
    translationSpace=BCS_BoneSpace
    alphaBoolBlend {blendInTime=0.3, blendOption=Linear, blendOutTime=0.4}
    boneToModify {boneName=dyn_hair_tail_1}
    rotation=<rot 5.846 -12.892 -0.116>
    translation=<0.656 -0.044 -0.96>
   }
   n15 = "Transform (Modify) Bone - Bone: dyn_hair_tail_1"(
    ComponentPose=@n14
    bAlphaBoolEnabled=Get Skydiving and Parachute not Opened() -> (Skydiving and Parachute not Opened:bool)
    Translation="0.597,3.492,-0.000"
    Rotation="-0.000,-24.503,-0.000"
    Scale="1,1,1"
   ) {
    alphaInputType=Bool
    rotationMode=BMM_Additive
    rotationSpace=BCS_BoneSpace
    translationMode=BMM_Additive
    translationSpace=BCS_BoneSpace
    alphaBoolBlend {blendInTime=0.2, blendOption=Linear, blendOutTime=0.2}
    boneToModify {boneName=dyn_hair_tail_1}
    rotation {pitch=0, yaw=-24.503}
    translation=<0.597 3.492>
   }
   n16 = "Transform (Modify) Bone - Bone: dyn_hair_tail_1"(
    ComponentPose=@n15
    bAlphaBoolEnabled=Get bIsParachuteOpened()
    Translation="-1.397,-0.242,-0.000"
    Rotation="0.000,14.596,-0.000"
    Scale="1,1,1"
   ) {
    alphaInputType=Bool
    rotationMode=BMM_Additive
    rotationSpace=BCS_BoneSpace
    translationMode=BMM_Additive
    translationSpace=BCS_BoneSpace
    alphaBoolBlend {blendInTime=0.2, blendOption=Linear, blendOutTime=0.2}
    boneToModify {boneName=dyn_hair_tail_1}
    rotation=<rot 0 14.596 0>
    translation=<-1.397 -0.242>
   }
   Output Pose(Result=Component To Local(ComponentPose=@n16)) {layerGroup=None}
   n17 = "Transform (Modify) Bone - Bone: dyn_hair_tail_1"!(
    ComponentPose
    Alpha:float=Get head_fwdBwd_rot_add() -> (head_fwdBwd_rot_add:float)
    Translation="-0.459,-1.375,-0.000"
    Rotation="-0.003,4.615,-0.003"
    Scale="1,1,1"
   ) -> (Pose) {
    rotationMode=BMM_Additive
    rotationSpace=BCS_BoneSpace
    translationMode=BMM_Additive
    translationSpace=BCS_BoneSpace
    alphaBoolBlend {blendOption=Linear}
    boneToModify {boneName=dyn_hair_tail_1}
    rotation=<rot -0.003 4.615 -0.003>
    translation=<-0.459 -1.375>
   }
   n18 = "Transform (Modify) Bone - Bone: dyn_hair_tail_1"(ComponentPose=@n17, Alpha=Get head_fwdBwd_rot(), Translation="-3.186,0.389,0.000", Rotation="-0.001,-12.385,0.276", Scale="1,1,1") {
    rotationMode=BMM_Additive
    rotationSpace=BCS_BoneSpace
    translationMode=BMM_Additive
    translationSpace=BCS_BoneSpace
    alphaBoolBlend {blendOption=Linear}
    alphaScaleBiasClamp {bMapRange=true, inRange {max=-89}}
    boneToModify {boneName=dyn_hair_tail_1}
    rotation=<rot -0.001 -12.385 0.276>
    translation=<-3.186 0.389>
   }
   n19 = "Transform (Modify) Bone - Bone: dyn_hair_tail_1"(ComponentPose=@n18, Alpha=Get head_fwdBwd_rot(), Translation="-0.415,0.312,-0.037", Rotation="0.005,12.529,-0.017", Scale="1,1,1") {
    rotationMode=BMM_Additive
    rotationSpace=BCS_BoneSpace
    translationMode=BMM_Additive
    translationSpace=BCS_BoneSpace
    alphaBoolBlend {blendOption=Linear}
    alphaScaleBiasClamp {bMapRange=true, inRange {max=16, min=-10}}
    boneToModify {boneName=dyn_hair_tail_1}
    rotation=<rot 0.005 12.529 -0.017>
    translation=<-0.415 0.312 -0.037>
   }
   Save cached pose 'SavedPose14171'(Pose=Component To Local(ComponentPose=@n19)) {cacheName=SavedPose14171}
   n20 = "Transform (Modify) Bone - Bone: dyn_hair_tail_1"(
    ComponentPose=Local To Component(LocalPose=Use cached pose 'SavedPose14171'())
    Alpha=Get head_fwdBwd_rot()
    Translation="2.094,1.828,0.000"
    Rotation="0.000,-23.439,-0.000"
    Scale="1,1,1"
   ) {
    rotationMode=BMM_Additive
    rotationSpace=BCS_BoneSpace
    translationMode=BMM_Additive
    translationSpace=BCS_BoneSpace
    alphaBoolBlend {blendOption=Linear}
    alphaScaleBiasClamp {bMapRange=true, inRange {max=23, min=-47}}
    boneToModify {boneName=dyn_hair_tail_1}
    rotation=<rot 0 -23.439 0>
    translation=<2.094 1.828>
   }
   n21 = "Transform (Modify) Bone - Bone: dyn_hair_tail_1"!(ComponentPose=@n20, bAlphaBoolEnabled:bool=True, Translation="-1.393,-1.879,0.000", Rotation="0.000,13.099,-0.000", Scale="1,1,1") -> (Pose) {
    alphaInputType=Bool
    rotationMode=BMM_Additive
    rotationSpace=BCS_BoneSpace
    translationMode=BMM_Additive
    translationSpace=BCS_BoneSpace
    alphaBoolBlend {blendOption=Linear}
    boneToModify {boneName=dyn_hair_tail_1}
    rotation=<rot 0 13.099 0>
    translation=<-1.393 -1.879>
   }
   n22 = "Transform (Modify) Bone - Bone: dyn_hair_tail_1"(
    ComponentPose=@n21
    bAlphaBoolEnabled=NOT Boolean(A=Get tactical sprinting and jumping or falling())
    Translation="0.862,0.416,-0.000"
    Rotation="0.048,1.881,-0.021"
    Scale="1,1,1"
   ) {
    alphaInputType=Bool
    rotationMode=BMM_Additive
    rotationSpace=BCS_BoneSpace
    translationMode=BMM_Additive
    translationSpace=BCS_BoneSpace
    alphaBoolBlend {blendOption=Linear}
    boneToModify {boneName=dyn_hair_tail_1}
    rotation=<rot 0.048 1.881 -0.021>
    translation=<0.862 0.416>
   }
   n23 = "Transform (Modify) Bone - Bone: dyn_hair_tail_1"!(ComponentPose=@n22, Alpha:float=Get head_fwdBwd_rot(), Translation="-0.114,-1.127,1.624", Rotation="0.001,6.419,-0.005", Scale="1,1,1") -> (Pose) {
    rotationMode=BMM_Additive
    rotationSpace=BCS_BoneSpace
    translationMode=BMM_Additive
    translationSpace=BCS_BoneSpace
    alphaBoolBlend {blendOption=Linear}
    alphaScaleBiasClamp {bMapRange=true, inRange {max=23}}
    boneToModify {boneName=dyn_hair_tail_1}
    rotation=<rot 0.001 6.419 -0.005>
    translation=<-0.114 -1.127 1.624>
   }
   n24 = "Transform (Modify) Bone - Bone: dyn_hair_tail_1"(ComponentPose=@n23, Alpha=Get head_fwdBwd_rot(), Translation="-0.895,-0.641,-0.010", Rotation="-1.348,-0.278,-6.430", Scale="1,1,1") {
    rotationMode=BMM_Additive
    rotationSpace=BCS_BoneSpace
    translationMode=BMM_Additive
    translationSpace=BCS_BoneSpace
    alphaBoolBlend {blendOption=Linear}
    alphaScaleBiasClamp {bMapRange=true, inRange {max=-42}}
    boneToModify {boneName=dyn_hair_tail_1}
    rotation=<rot -1.348 -0.278 -6.43>
    translation=<-0.895 -0.641 -0.01>
   }
   n25 = "Transform (Modify) Bone - Bone: dyn_hair_tail_1"(ComponentPose=@n24, Alpha=Get head_fwdBwd_rot(), Translation="0.405,0.700,-0.000", Rotation="0,0,0", Scale="1,1,1") {
    rotationMode=BMM_Additive
    rotationSpace=BCS_BoneSpace
    translationMode=BMM_Additive
    translationSpace=BCS_BoneSpace
    alphaBoolBlend {blendOption=Linear}
    alphaScaleBiasClamp {bMapRange=true, inRange {max=-16, min=5}}
    boneToModify {boneName=dyn_hair_tail_1}
    translation=<0.405 0.7>
   }
   n26 = Component To Local(ComponentPose=@n25)
   n27 = Use cached pose 'SavedPose14171'()
   n28 = Get bIsCrouching()
   n29 = "Transform (Modify) Bone - Bone: dyn_hair_tail_1"!(
    ComponentPose=Local To Component(LocalPose=Blend Poses by bool(BlendPose_0=@n27, BlendPose_1=@n26, bActiveValue=@n28) {blendType=Linear})
    bAlphaBoolEnabled:bool=Get tactical sprinting and jumping or falling()
    Translation="-1.306,0.797,-0.122"
    Rotation="-1.512,10.268,2.876"
    Scale="1,1,1"
   ) -> (Pose) {
    alphaInputType=Bool
    rotationMode=BMM_Additive
    rotationSpace=BCS_BoneSpace
    translationMode=BMM_Additive
    translationSpace=BCS_BoneSpace
    alphaBoolBlend {blendInTime=0.2, blendOption=Linear, blendOutTime=0.2}
    boneToModify {boneName=dyn_hair_tail_1}
    rotation=<rot -1.512 10.268 2.876>
    translation=<-1.306 0.797 -0.122>
   }
   n30 = "Transform (Modify) Bone - Bone: dyn_hair_tail_1"(ComponentPose=@n29, bAlphaBoolEnabled=Get bIsJumping(), Translation="0.656,-0.044,-0.960", Rotation="5.846,-12.892,-0.116", Scale="1,1,1") {
    alphaInputType=Bool
    rotationMode=BMM_Additive
    rotationSpace=BCS_BoneSpace
    translationMode=BMM_Additive
    translationSpace=BCS_BoneSpace
    alphaBoolBlend {blendInTime=0.3, blendOption=Linear, blendOutTime=0.4}
    boneToModify {boneName=dyn_hair_tail_1}
    rotation=<rot 5.846 -12.892 -0.116>
    translation=<0.656 -0.044 -0.96>
   }
   n31 = "Transform (Modify) Bone - Bone: dyn_hair_tail_1"(ComponentPose=@n30, bAlphaBoolEnabled=Get bIsFalling(), Translation="0.656,-0.044,-0.96", Rotation="5.846,-12.892,-0.116", Scale="1,1,1") {
    alphaInputType=Bool
    rotationMode=BMM_Additive
    rotationSpace=BCS_BoneSpace
    translationMode=BMM_Additive
    translationSpace=BCS_BoneSpace
    alphaBoolBlend {blendInTime=0.3, blendOption=Linear, blendOutTime=0.4}
    boneToModify {boneName=dyn_hair_tail_1}
    rotation=<rot 5.846 -12.892 -0.116>
    translation=<0.656 -0.044 -0.96>
   }
   n32 = "Transform (Modify) Bone - Bone: dyn_hair_tail_1"(
    ComponentPose=@n31
    bAlphaBoolEnabled=Get Target and2d moving() -> (Target and2d moving:bool)
    Translation="-0.845,-0.649,0.000"
    Rotation="0.000,9.285,-0.000"
    Scale="1,1,1"
   ) {
    alphaInputType=Bool
    rotationMode=BMM_Additive
    rotationSpace=BCS_BoneSpace
    translationMode=BMM_Additive
    translationSpace=BCS_BoneSpace
    alphaBoolBlend {blendInTime=0.2, blendOption=Linear, blendOutTime=0.2}
    boneToModify {boneName=dyn_hair_tail_1}
    rotation=<rot 0 9.2851 0>
    translation=<-0.845 -0.649>
   }
   n33 = "Transform (Modify) Bone - Bone: dyn_hair_tail_1"(
    ComponentPose=@n32
    bAlphaBoolEnabled=Get Targeting not crouching() -> (Targeting not crouching:bool)
    Translation="-0.169,0.232,0.000"
    Rotation="-0.005,-6.482,-0.017"
    Scale="1,1,1"
   ) {
    alphaInputType=Bool
    rotationMode=BMM_Additive
    rotationSpace=BCS_BoneSpace
    translationMode=BMM_Additive
    translationSpace=BCS_BoneSpace
    alphaBoolBlend {blendInTime=0.2, blendOption=Linear, blendOutTime=0.2}
    boneToModify {boneName=dyn_hair_tail_1}
    rotation=<rot -0.005 -6.482 -0.017>
    translation=<-0.169 0.232>
   }
   n34 = "Transform (Modify) Bone - Bone: dyn_hair_tail_1"(
    ComponentPose=@n33
    bAlphaBoolEnabled=Get bIsTacticalSprinting()
    Translation="1.008,1.541,0.000"
    Rotation="-0.000,-26.564,-0.000"
    Scale="1,1,1"
   ) {
    alphaInputType=Bool
    rotationMode=BMM_Additive
    rotationSpace=BCS_BoneSpace
    translationMode=BMM_Additive
    translationSpace=BCS_BoneSpace
    alphaBoolBlend {blendInTime=0.2, blendOption=Linear, blendOutTime=0.2}
    boneToModify {boneName=dyn_hair_tail_1}
    rotation {yaw=-26.564}
    translation=<1.008 1.541>
   }
   n35 = "Transform (Modify) Bone - Bone: dyn_hair_tail_1"(
    ComponentPose=@n34
    bAlphaBoolEnabled=Get Skydiving and Parachute not Opened()
    Translation="0.597,3.492,-0.000"
    Rotation="-0.000,-24.503,-0.000"
    Scale="1,1,1"
   ) {
    alphaInputType=Bool
    rotationMode=BMM_Additive
    rotationSpace=BCS_BoneSpace
    translationMode=BMM_Additive
    translationSpace=BCS_BoneSpace
    alphaBoolBlend {blendInTime=0.2, blendOption=Linear, blendOutTime=0.2}
    boneToModify {boneName=dyn_hair_tail_1}
    rotation=<rot 0 -24.503 0>
    translation=<0.597 3.492>
   }
   Save cached pose 'SavedPose23283'(Pose=Component To Local(ComponentPose=@n35)) {cacheName=SavedPose23283}
  ]}
 EventGraph {nodes [
   Event BlueprintUpdateAnimation()
   > n1 = Sequence!(execute) -> (then_0, then_1, then_2:Exec, then_3:Exec, then_4:Exec)
   n2 = Gravity Override State Function DEPRECATED(
    execute=@n1.then_1
    CharacterPartAnimInstance=Self-Reference()
    GameTimeInSeconds=GetGameTimeInSeconds()
    GravityOverrideParams=Get Gravity Override Params()
   )
   > Set gravity_override(gravity_override=@n2)
   n3 = Get bIsJumping()
   n4 = Get bIsFalling()
   n5 = Get bIsTacticalSprinting()
   n6 = AND Boolean(A=@n5, B=OR Boolean(A=@n3, B=@n4))
   Set tactical sprinting and jumping or falling(execute=@n1.then_2, tactical sprinting and jumping or falling:bool=@n6) -> (Output_Get:bool)
   > n7 = Branch(Condition=@n6)
   > Set Alpha(Alpha:float=0) -> (Output_Get:float)
   n8 = Branch(execute=@n7.else, Condition=OR Boolean!(A=@n5, B=@n3, C:bool=@n4) -> (ReturnValue))
   > Set Alpha(Alpha=0.3)
   Set Alpha(execute=@n8.else, Alpha=1)
   n9 = Get WeaponCoreAnim!() -> (WeaponCoreAnim)
   n10 = OR Boolean(
    A="Equal (Enum)"(A:EFortWeaponCoreAnimation Enum=@n9, B:EFortWeaponCoreAnimation Enum=Melee)
    B="Equal (Enum)"(A=@n9, B=Pistol)
    C="Equal (Enum)"(A=@n9, B=MeleeDualWield)
   )
   Set Melee and pistol Weapons(
    execute=@n1.then_4
    Melee and pistol Weapons:double=FInterpTo(
     Current=Get Melee and pistol Weapons()
     Target="To Float (Boolean)"(InBool=@n10)
     DeltaTime=GetWorldDeltaSeconds!() -> (ReturnValue:double)
     InterpSpeed=20
    )
   ) -> (Output_Get:double)
   n11 = Get bIsCrouching()
   > Set CrouchingAlpha(CrouchingAlpha:double=FInterpTo(Current=Get CrouchingAlpha(), Target="To Float (Boolean)"(InBool=@n11), DeltaTime=GetWorldDeltaSeconds(), InterpSpeed=25)) -> (Output_Get:double)
   n12 = Get bIsTargeting()
   > Set Targeting alpha(Targeting alpha:double=FInterpTo(Current=Get Targeting alpha(), Target="To Float (Boolean)"(InBool=@n12), DeltaTime=GetWorldDeltaSeconds(), InterpSpeed=25)) -> (Output_Get:double)
   > Set Crouching and moving(Crouching and moving:bool=AND Boolean(A=Get bIsCrouching(), B=Get bIsMoving2D())) -> (Output_Get:bool)
   n13 = Get bIsCrouching()
   > Set Not crouching and moving(Not crouching and moving:bool=AND Boolean(A=Get bIsMoving2D(), B=NOT Boolean(A=@n13))) -> (Output_Get:bool)
   n14 = Get bIsCrouching()
   Set Targeting not crouching(execute=@n1.then_3, Targeting not crouching:bool=AND Boolean(A=Get bIsTargeting(), B=NOT Boolean(A=@n14))) -> (Output_Get:bool)
   n15 = OR Boolean(A=Get bIsCrouching(), B=Get bIsTargeting(), C=Get bIsMoving2D())
   > Set is in idle(is in idle:bool=AND Boolean(A=NOT Boolean(A=@n15), B=Get bIsOnGround() -> (bIsOnGround:bool))) -> (Output_Get:bool)
   n16 = Get bIsParachuteOpened()
   > Set Skydiving and Parachute not Opened(Skydiving and Parachute not Opened:bool=AND Boolean(A=Get bIsSkyDiving(), B=NOT Boolean(A=@n16))) -> (Output_Get:bool)
   > Set Target and2d moving(Target and2d moving:bool=AND Boolean!(A=Get bIsTargeting(), B=Get bIsMoving2D(), C:bool=Get bIsCrouching()) -> (ReturnValue)) -> (Output_Get:bool)
   n17 = GetDeltaTransformFromRefPose(self=GetOwnerSkeletalMeshComponent(), BoneName=head, BaseName=spine_05)
   Set head_fwdBwd_rot(execute=@n1.then_0, head_fwdBwd_rot:double=BreakRotator(InRot=BreakTransform(InTransform=@n17).Rotation).Yaw) -> (Output_Get:double)
   n18 = BreakTransform(InTransform=GetDeltaTransformFromRefPose(self=GetOwnerSkeletalMeshComponent(), BoneName=head, BaseName=neck_02))
   > Set head_fwdBwd_rot_add(head_fwdBwd_rot_add:float=MapRangeClamped(Value=BreakRotator(InRot=@n18.Rotation).Yaw, InRangeA=-28, InRangeB=-43)) -> (Output_Get:float)
  ]}
]
variables [
 Alpha {type=number, default=1}
 Crouching and moving {type=boolean}
 CrouchingAlpha {type=number}
 Gravity Override Params {type=GravityOverrideParamsStruct}
 gravity_override {type=Vector}
 head_fwdBwd_rot {type=number}
 head_fwdBwd_rot_add {type=number}
 is in idle {type=boolean}
 Melee and pistol Weapons {type=number}
 Not crouching and moving {type=boolean}
 Skydiving and Parachute not Opened {type=boolean}
 tactical sprinting and jumping or falling {type=boolean}
 Target and2d moving {type=boolean}
 Targeting alpha {type=number}
 Targeting not crouching {type=boolean}
]
